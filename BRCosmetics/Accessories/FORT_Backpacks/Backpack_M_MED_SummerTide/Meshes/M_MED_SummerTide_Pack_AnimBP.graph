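# Animation Blueprint for the SummerTide backpack, adding physical jiggle to Peely, sunglasses, ice cubes and umbrella.
# AnimDynamics water bones drive Peely and ice-cube bases through weighted constraints, then ice cubes simulate into a cached pose.
# Two differently tuned Peely/sunglasses chains read that cache into an unwired bool blend, followed by umbrella and cloth dynamics.
# Final pack-bone offsets follow gender, female skydive floating and spine bend computed in the EventGraph.
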AnimBlueprint /BRCosmetics/Accessories/FORT_Backpacks/Backpack_M_MED_SummerTide/Meshes/M_MED_SummerTide_Pack_AnimBP
target_skeleton=&/M_MED_SummerTide_Pack_Skeleton
graphs [
 AnimGraph {nodes [
   n1 = Anim Dynamics - Bone: dyn_peely!(ComponentPose=Local To Component(LocalPose=Use cached pose 'PreAnimDym'()), Alpha=1) -> (Pose) {
    angularDampingOverride=1
    angularSpringConstant=25000
    gravityScale=0
    linearSpringConstant=1
    alphaBoolBlend {blendOption=Linear}
    boundBone {boneName=dyn_peely}
    componentAppliedLinearAccClamp=<3000 3000 3000>
    componentLinearAccScale=<0.1 0.1 0.1>
    componentLinearVelScale=<0.1 0.1 0.1>
    physicsBodyDefinitions [{
      boundBone {boneName=dyn_peely}
      boxExtents=<10 10 10>
      constraintSetup {angularTargetAxis=AxisY, twistAxis=AxisY, angularLimitsMax=<0 0 10>, angularLimitsMin=<0 0 -10>, angularTarget=<0 1 0>}
      localJointOffset=<0 1 0>
     }]
   }
   n2 = Anim Dynamics - Bone: dyn_peely(ComponentPose=@n1, Alpha=1) {
    angularDampingOverride=1
    angularSpringConstant=2000
    bAngularSpring=true
    bLinearSpring=true
    bOverrideAngularDamping=true
    bOverrideLinearDamping=true
    gravityScale=0
    linearDampingOverride=1
    linearSpringConstant=100
    simulationSpace=RootRelative
    alphaBoolBlend {blendOption=Linear}
    boundBone {boneName=dyn_peely}
    componentAppliedLinearAccClamp=<3000 3000 3000>
    componentLinearAccScale=<0.2 0.2 0.1>
    componentLinearVelScale=<0.2 0.2 0.1>
    physicsBodyDefinitions [{
      boundBone {boneName=dyn_peely}
      boxExtents=<10 10 10>
      constraintSetup {angularLimitsMax=<10 10 10>, angularLimitsMin=<-10 -10 -10>, angularTarget=<1 0 0>}
      localJointOffset=<0 0 -1>
     }]
    relativeSpaceBone {boneName=root}
   }
   n3 = Anim Dynamics - Bone: dyn_peely(ComponentPose=@n2, Alpha=1) {
    angularDampingOverride=1
    angularSpringConstant=25000
    bAngularSpring=true
    bLinearSpring=true
    bOverrideAngularDamping=true
    bOverrideLinearDamping=true
    gravityScale=0
    linearDampingOverride=1
    linearSpringConstant=100
    simulationSpace=RootRelative
    alphaBoolBlend {blendOption=Linear}
    boundBone {boneName=dyn_peely}
    componentAppliedLinearAccClamp=<3000 3000 3000>
    componentLinearAccScale=<0.2 0.2 0.1>
    componentLinearVelScale=<0.2 0.2 0.1>
    physicsBodyDefinitions [{
      boundBone {boneName=dyn_peely}
      boxExtents=<10 10 10>
      constraintSetup {angularLimitsMax=<0 0 0>, angularLimitsMin=<0 0 0>, angularTarget=<1 0 0>, linearAxesMax=<0 0 1>, linearAxesMin=<0 0 -1>}
      localJointOffset=<0 0 0>
     }]
    relativeSpaceBone {boneName=root}
   }
   n4 = Anim Dynamics - Bone: dyn_peelySunGlasses!(ComponentPose=@n3, Alpha=1, GravityOverride:Vector="0,0,-980") -> (Pose) {
    angularDampingOverride=0.9999
    angularSpringConstant=12000
    bAngularSpring=true
    bOverrideAngularDamping=true
    bUseGravityOverride=true
    gravityScale=0.35
    alphaBoolBlend {blendOption=Linear}
    boundBone {boneName=dyn_peelySunGlasses}
    componentAppliedLinearAccClamp=<250 250 250>
    componentLinearAccScale=<1 1 1>
    componentLinearVelScale=<20 20 20>
    gravityOverride {z=-980}
    physicsBodyDefinitions [{
      boundBone {boneName=dyn_peelySunGlasses}
      boxExtents=<10 10 10>
      constraintSetup {angularLimitsMax=<0 5 0>, angularLimitsMin=<0 -5 0>, angularTarget=<1 0 0>}
      localJointOffset=<1 0 0>
     }]
    relativeSpaceBone {boneName=base_peelySunGlasses}
   }
   n5 = Component To Local(ComponentPose=@n4)
   n6 = Anim Dynamics - Bone: dyn_peely(ComponentPose=Local To Component(LocalPose=Use cached pose 'PreAnimDym'()), Alpha=1) {
    angularDampingOverride=1
    angularSpringConstant=25000
    gravityScale=0
    linearSpringConstant=1
    alphaBoolBlend {blendOption=Linear}
    boundBone {boneName=dyn_peely}
    componentAppliedLinearAccClamp=<3000 3000 3000>
    componentLinearAccScale=<0.1 0.1 0.1>
    componentLinearVelScale=<0.1 0.1 0.1>
    physicsBodyDefinitions [{
      boundBone {boneName=dyn_peely}
      boxExtents=<10 10 10>
      constraintSetup {angularTargetAxis=AxisY, twistAxis=AxisY, angularLimitsMax=<0 0 10>, angularLimitsMin=<0 0 -10>, angularTarget=<0 1 0>}
      localJointOffset=<0 1 0>
     }]
   }
   n7 = Anim Dynamics - Bone: dyn_peely(ComponentPose=@n6, Alpha=1) {
    angularDampingOverride=1
    angularSpringConstant=25000
    bAngularSpring=true
    bLinearSpring=true
    bOverrideAngularDamping=true
    bOverrideLinearDamping=true
    gravityScale=0
    linearDampingOverride=1
    linearSpringConstant=100
    simulationSpace=RootRelative
    alphaBoolBlend {blendOption=Linear}
    boundBone {boneName=dyn_peely}
    componentAppliedLinearAccClamp=<3000 3000 3000>
    componentLinearAccScale=<0.2 0.2 0.1>
    componentLinearVelScale=<0.2 0.2 0.1>
    physicsBodyDefinitions [{
      boundBone {boneName=dyn_peely}
      boxExtents=<10 10 10>
      constraintSetup {angularLimitsMax=<10 10 10>, angularLimitsMin=<-10 -10 -10>, angularTarget=<1 0 0>}
      localJointOffset=<0 0 -1>
     }]
    relativeSpaceBone {boneName=root}
   }
   n8 = Anim Dynamics - Bone: dyn_peely(ComponentPose=@n7, Alpha=1) {
    angularDampingOverride=1
    angularSpringConstant=25000
    bAngularSpring=true
    bLinearSpring=true
    bOverrideAngularDamping=true
    bOverrideLinearDamping=true
    gravityScale=0
    linearDampingOverride=1
    linearSpringConstant=100
    simulationSpace=RootRelative
    alphaBoolBlend {blendOption=Linear}
    boundBone {boneName=dyn_peely}
    componentAppliedLinearAccClamp=<3000 3000 3000>
    componentLinearAccScale=<0.2 0.2 0.1>
    componentLinearVelScale=<0.2 0.2 0.1>
    physicsBodyDefinitions [{
      boundBone {boneName=dyn_peely}
      boxExtents=<10 10 10>
      constraintSetup {angularLimitsMax=<0 0 0>, angularLimitsMin=<0 0 0>, angularTarget=<1 0 0>, linearAxesMax=<0 0 1>, linearAxesMin=<0 0 -1>}
      localJointOffset=<0 0 0>
     }]
    relativeSpaceBone {boneName=root}
   }
   n9 = Anim Dynamics - Bone: dyn_peelySunGlasses(ComponentPose=@n8, Alpha=1, GravityOverride="0,0,-980") {
    angularDampingOverride=0.9999
    angularSpringConstant=20000
    bAngularSpring=true
    bOverrideAngularDamping=true
    bUseGravityOverride=true
    gravityScale=0.35
    alphaBoolBlend {blendOption=Linear}
    boundBone {boneName=dyn_peelySunGlasses}
    componentAppliedLinearAccClamp=<250 250 250>
    componentLinearAccScale=<1 1 1>
    componentLinearVelScale=<20 20 20>
    gravityOverride {z=-980}
    physicsBodyDefinitions [{
      boundBone {boneName=dyn_peelySunGlasses}
      boxExtents=<10 10 10>
      constraintSetup {angularLimitsMax=<0 5 0>, angularLimitsMin=<0 -5 0>, angularTarget=<1 0 0>}
      localJointOffset=<1 0 0>
     }]
    relativeSpaceBone {boneName=base_peelySunGlasses}
   }
   n10 = Component To Local(ComponentPose=@n9)
   n11 = Anim Dynamics - Bone: dyn_umbrella!(
    ComponentPose=Local To Component(LocalPose=Blend Poses by bool(BlendPose_0=@n5, BlendPose_1=@n10) {blendType=Linear})
    Alpha=1
    GravityOverride:Vector="0,0,-980"
    ExternalForce="0,0,0"
   ) -> (Pose) {
    angularDampingOverride=1
    angularSpringConstant=20000
    bAngularSpring=true
    bOverrideAngularDamping=true
    bUseGravityOverride=true
    gravityScale=0.35
    linearDampingOverride=1
    alphaBoolBlend {blendOption=Linear}
    boundBone {boneName=dyn_umbrella}
    componentAppliedLinearAccClamp=<500 500 500>
    componentLinearAccScale=<0.1 0.1 0.1>
    componentLinearVelScale {x=0.1, z=0.1}
    gravityOverride {z=-980}
    physicsBodyDefinitions [{
      boundBone {boneName=dyn_umbrella}
      boxExtents=<20 30 30>
      constraintSetup {angularLimitsMax=<30 1 0>, angularLimitsMin=<-30 -5 0>, angularTarget=<1 0 0>}
      localJointOffset=<4 0 0>
     }]
    relativeSpaceBone {boneName=base_umbrella}
   }
   n12 = Anim Dynamics - Bone: dyn_umbrellaCloth_1!(ComponentPose=@n11, Alpha=1) -> (Pose) {
    angularDampingOverride=1
    angularSpringConstant=20000
    bAngularSpring=true
    bOverrideAngularDamping=true
    gravityScale=0
    alphaBoolBlend {blendOption=Linear}
    boundBone {boneName=dyn_umbrellaCloth_1}
    componentAppliedLinearAccClamp=<3000 3000 3000>
    componentLinearAccScale=<1 1 1>
    componentLinearVelScale=<10 10 10>
    gravityOverride {z=-980}
    physicsBodyDefinitions [{
      boundBone {boneName=dyn_umbrellaCloth_1}
      boxExtents=<20 30 30>
      constraintSetup {angularLimitsMax=<0 15 0>, angularLimitsMin=<0 -15 0>, angularTarget=<1 0 0>}
      localJointOffset=<4 0 0>
     }]
    relativeSpaceBone {boneName=base_umbrellaCloth_1}
   }
   n13 = Anim Dynamics - Bone: dyn_umbrellaCloth_2!(ComponentPose=@n12, Alpha=1) -> (Pose) {
    angularDampingOverride=1
    angularSpringConstant=20000
    bAngularSpring=true
    bOverrideAngularDamping=true
    gravityScale=0
    linearSpringConstant=250
    alphaBoolBlend {blendOption=Linear}
    boundBone {boneName=dyn_umbrellaCloth_2}
    componentAppliedLinearAccClamp=<3000 3000 3000>
    componentLinearAccScale=<1 1 1>
    componentLinearVelScale=<10 10 10>
    gravityOverride {z=-980}
    physicsBodyDefinitions [{
      boundBone {boneName=dyn_umbrellaCloth_2}
      boxExtents=<20 30 30>
      constraintSetup {angularLimitsMax=<0 15 0>, angularLimitsMin=<0 -15 0>, angularTarget=<1 0 0>}
      localJointOffset=<4 0 0>
     }]
    relativeSpaceBone {boneName=base_umbrellaCloth_2}
   }
   n14 = Anim Dynamics - Bone: dyn_umbrellaCloth_3!(ComponentPose=@n13, Alpha=1) -> (Pose) {
    angularDampingOverride=1
    angularSpringConstant=20000
    bAngularSpring=true
    bOverrideAngularDamping=true
    gravityScale=0
    linearSpringConstant=250
    alphaBoolBlend {blendOption=Linear}
    boundBone {boneName=dyn_umbrellaCloth_3}
    componentAppliedLinearAccClamp=<3000 3000 3000>
    componentLinearAccScale=<1 1 1>
    componentLinearVelScale=<10 10 10>
    gravityOverride {z=-980}
    physicsBodyDefinitions [{
      boundBone {boneName=dyn_umbrellaCloth_3}
      boxExtents=<20 30 30>
      constraintSetup {angularLimitsMax=<0 15 0>, angularLimitsMin=<0 -15 0>, angularTarget=<1 0 0>}
      localJointOffset=<4 0 0>
     }]
    relativeSpaceBone {boneName=base_umbrellaCloth_3}
   }
   n15 = Anim Dynamics - Bone: dyn_umbrellaCloth_4!(ComponentPose=@n14, Alpha=1) -> (Pose) {
    angularDampingOverride=1
    angularSpringConstant=20000
    bAngularSpring=true
    bOverrideAngularDamping=true
    gravityScale=0
    linearSpringConstant=250
    alphaBoolBlend {blendOption=Linear}
    boundBone {boneName=dyn_umbrellaCloth_4}
    componentAppliedLinearAccClamp=<3000 3000 3000>
    componentLinearAccScale=<1 1 1>
    componentLinearVelScale=<10 10 10>
    gravityOverride {z=-980}
    physicsBodyDefinitions [{
      boundBone {boneName=dyn_umbrellaCloth_4}
      boxExtents=<20 30 30>
      constraintSetup {angularLimitsMax=<0 15 0>, angularLimitsMin=<0 -15 0>, angularTarget=<1 0 0>}
      localJointOffset=<4 0 0>
     }]
    relativeSpaceBone {boneName=base_umbrellaCloth_4}
   }
   n16 = "Transform (Modify) Bone - Bone: pack"!(ComponentPose=@n15, bAlphaBoolEnabled:bool=Get isMale() -> (isMale:bool), Translation="0,-0.6,0") -> (Pose) {
    alphaInputType=Bool
    translationMode=BMM_Additive
    translationSpace=BCS_BoneSpace
    alphaBoolBlend {blendOption=Linear}
    boneToModify {boneName=pack}
    translation {y=-0.6}
   }
   n17 = "Transform (Modify) Bone - Bone: pack"!(ComponentPose=@n16, Alpha:float, Translation=Get backpack_translation(), Rotation:Rotator=Get backpack_rotate()) -> (Pose) {translationMode=BMM_Additive, translationSpace=BCS_BoneSpace, alphaBoolBlend {blendOption=Linear}, boneToModify {boneName=pack}}
   n18 = "Transform (Modify) Bone - Bone: pack"!(ComponentPose=@n17, bAlphaBoolEnabled:bool=Get isFemaleAndFloating() -> (isFemaleAndFloating:bool), Translation="0,-0.55,0") -> (Pose) {
    alphaInputType=Bool
    translationMode=BMM_Additive
    translationSpace=BCS_BoneSpace
    alphaBoolBlend {blendOption=Linear}
    boneToModify {boneName=pack}
    translation {y=-0.55}
   }
   Output Pose(Result=Component To Local(ComponentPose=@n18))
   n19 = Anim Dynamics - Bone: dyn_water_T_1!(ComponentPose=Local To Component(LocalPose=Input Pose()), Alpha=1) -> (Pose) {
    angularSpringConstant=200
    bAngularSpring=true
    bGravityOverrideInSimSpace=true
    bLinearSpring=true
    bOverrideLinearDamping=true
    gravityScale=0
    linearDampingOverride=1
    linearSpringConstant=20
    simulationSpace=RootRelative
    alphaBoolBlend {blendOption=Linear}
    alphaScaleBiasClamp {interpSpeedDecreasing=20, interpSpeedIncreasing=20}
    boundBone {boneName=dyn_water_T_1}
    componentAppliedLinearAccClamp=<150 150 150>
    componentLinearAccScale=<1 1 1>
    componentLinearVelScale=<20 20 20>
    physicsBodyDefinitions [{
      sphereCollisionRadius=0
      boundBone {boneName=dyn_water_T_1}
      boxExtents=<10 10 10>
      constraintSetup {angularLimitsMax=<0 4 3>, angularLimitsMin=<0 -1 -3>, angularTarget=<1 0 0>, linearAxesMax=<0 0 0.2>, linearAxesMin=<0 0 -0.2>}
      localJointOffset=<10 0 0>
     }]
   }
   n20 = Anim Dynamics - Bone: dyn_water_R_1!(ComponentPose=@n19, Alpha=1) -> (Pose) {
    angularSpringConstant=200
    bAngularSpring=true
    bGravityOverrideInSimSpace=true
    bLinearSpring=true
    bOverrideLinearDamping=true
    gravityScale=0
    linearDampingOverride=1
    linearSpringConstant=20
    simulationSpace=RootRelative
    alphaBoolBlend {blendOption=Linear}
    alphaScaleBiasClamp {interpSpeedDecreasing=20, interpSpeedIncreasing=20}
    boundBone {boneName=dyn_water_R_1}
    componentAppliedLinearAccClamp=<150 150 150>
    componentLinearAccScale=<1 1 1>
    componentLinearVelScale=<20 20 20>
    physicsBodyDefinitions [{
      sphereCollisionRadius=0
      boundBone {boneName=dyn_water_R_1}
      boxExtents=<10 10 10>
      constraintSetup {angularLimitsMax=<0 4 3>, angularLimitsMin=<0 -1 -3>, angularTarget=<1 0 0>, linearAxesMax=<0 0 0.2>, linearAxesMin=<0 0 -0.2>}
      localJointOffset=<-10 0 0>
     }]
   }
   n21 = "Constraint - base_peely (PP)"!(ComponentPose=@n20, Alpha=1, ConstraintWeights_0=0.7026, ConstraintWeights_1:float=0.2974) -> (Pose) {
    alphaBoolBlend {blendOption=Linear}
    boneToModify {boneName=base_peely}
    constraintSetup [{targetBone {boneName=dyn_water_T_1}}, {targetBone {boneName=dyn_water_R_1}}]
    constraintWeights [1,1]
   }
   n22 = "Constraint - base_iceCube_2 (PP)"!(ComponentPose=@n21, Alpha=1, ConstraintWeights_0=0.7013, ConstraintWeights_1:float=0.2987) -> (Pose) {
    alphaBoolBlend {blendOption=Linear}
    boneToModify {boneName=base_iceCube_2}
    constraintSetup [{targetBone {boneName=dyn_water_R_1}}, {targetBone {boneName=dyn_water_T_1}}]
    constraintWeights [1,1]
   }
   n23 = "Constraint - base_iceCube_1 (PP)"!(ComponentPose=@n22, Alpha=1, ConstraintWeights_0=0.8819, ConstraintWeights_1:float=0.1181) -> (Pose) {
    alphaBoolBlend {blendOption=Linear}
    boneToModify {boneName=base_iceCube_1}
    constraintSetup [{targetBone {boneName=dyn_water_R_1}}, {targetBone {boneName=dyn_water_T_1}}]
    constraintWeights [1,1]
   }
   n24 = "Constraint - base_iceCube_3 (PP)"!(ComponentPose=@n23, Alpha=1, ConstraintWeights_0=0.8828, ConstraintWeights_1:float=0.1172) -> (Pose) {
    alphaBoolBlend {blendOption=Linear}
    boneToModify {boneName=base_iceCube_3}
    constraintSetup [{targetBone {boneName=dyn_water_R_1}}, {targetBone {boneName=dyn_water_T_1}}]
    constraintWeights [1,1]
   }
   n25 = "Constraint - base_iceCube_4 (PP)"!(ComponentPose=@n24, Alpha=1, ConstraintWeights_0=0.8817, ConstraintWeights_1:float=0.1183) -> (Pose) {
    alphaBoolBlend {blendOption=Linear}
    boneToModify {boneName=base_iceCube_4}
    constraintSetup [{targetBone {boneName=dyn_water_R_1}}, {targetBone {boneName=dyn_water_T_1}}]
    constraintWeights [1,1]
   }
   n26 = "Constraint - base_iceCube_5 (PP)"!(ComponentPose=@n25, Alpha=1, ConstraintWeights_0=0.7001, ConstraintWeights_1:float=0.2999) -> (Pose) {
    alphaBoolBlend {blendOption=Linear}
    boneToModify {boneName=base_iceCube_5}
    constraintSetup [{targetBone {boneName=dyn_water_R_1}}, {targetBone {boneName=dyn_water_T_1}}]
    constraintWeights [1,1]
   }
   n27 = "Constraint - base_iceCube_6 (PP)"!(ComponentPose=@n26, Alpha=1, ConstraintWeights_0=0.5045, ConstraintWeights_1:float=0.4955) -> (Pose) {
    alphaBoolBlend {blendOption=Linear}
    boneToModify {boneName=base_iceCube_6}
    constraintSetup [{targetBone {boneName=dyn_water_T_1}}, {targetBone {boneName=dyn_water_R_1}}]
    constraintWeights [1,1]
   }
   n28 = "Constraint - base_iceCube_7 (PP)"!(ComponentPose=@n27, Alpha=1, ConstraintWeights_0=0.7014, ConstraintWeights_1:float=0.2986) -> (Pose) {
    alphaBoolBlend {blendOption=Linear}
    boneToModify {boneName=base_iceCube_7}
    constraintSetup [{targetBone {boneName=dyn_water_T_1}}, {targetBone {boneName=dyn_water_R_1}}]
    constraintWeights [1,1]
   }
   n29 = Anim Dynamics - Bone: dyn_iceCube_1!(ComponentPose=@n28, Alpha=1) -> (Pose) {
    angularDampingOverride=1
    angularSpringConstant=200
    bAngularSpring=true
    bLinearSpring=true
    bOverrideLinearDamping=true
    gravityScale=0
    linearDampingOverride=1
    linearSpringConstant=200
    simulationSpace=RootRelative
    alphaBoolBlend {blendOption=Linear}
    boundBone {boneName=dyn_iceCube_1}
    componentAppliedLinearAccClamp=<10000 10000 10000>
    componentLinearAccScale=<0.3 0.3 0.2>
    componentLinearVelScale=<0.3 0.3 0.2>
    physicsBodyDefinitions [{
      boundBone {boneName=dyn_iceCube_1}
      boxExtents=<15 15 15>
      constraintSetup {angularLimitsMax=<10 10 10>, angularLimitsMin=<-10 -10 -10>, angularTarget=<1 0 0>, linearAxesMax=<0 0 2>, linearAxesMin=<0 0 -0.5>}
      localJointOffset=<1 0 0>
     }]
    relativeSpaceBone {boneName=root}
   }
   n30 = Anim Dynamics - Bone: dyn_iceCube_2!(ComponentPose=@n29, Alpha=1) -> (Pose) {
    angularDampingOverride=1
    angularSpringConstant=200
    bAngularSpring=true
    bLinearSpring=true
    bOverrideLinearDamping=true
    gravityScale=0
    linearDampingOverride=1
    linearSpringConstant=200
    simulationSpace=RootRelative
    alphaBoolBlend {blendOption=Linear}
    boundBone {boneName=dyn_iceCube_2}
    componentAppliedLinearAccClamp=<200 200 200>
    componentLinearAccScale=<1 1 1>
    componentLinearVelScale=<20 20 20>
    physicsBodyDefinitions [{
      boundBone {boneName=dyn_iceCube_2}
      boxExtents=<10 10 10>
      constraintSetup {angularLimitsMax=<10 10 10>, angularLimitsMin=<-10 -10 -10>, angularTarget=<1 0 0>, linearAxesMax=<0 0 1>, linearAxesMin=<0 0 -0.5>}
      localJointOffset=<1 0.5 0>
     }]
    relativeSpaceBone {boneName=root}
   }
   n31 = Anim Dynamics - Bone: dyn_iceCube_3!(ComponentPose=@n30, Alpha=1) -> (Pose) {
    angularDampingOverride=1
    angularSpringConstant=200
    bAngularSpring=true
    bLinearSpring=true
    bOverrideLinearDamping=true
    gravityScale=0
    linearDampingOverride=1
    linearSpringConstant=200
    simulationSpace=RootRelative
    alphaBoolBlend {blendOption=Linear}
    boundBone {boneName=dyn_iceCube_3}
    componentAppliedLinearAccClamp=<200 200 200>
    componentLinearAccScale=<1 1 1>
    componentLinearVelScale=<20 20 20>
    physicsBodyDefinitions [{
      boundBone {boneName=dyn_iceCube_3}
      boxExtents=<10 10 10>
      constraintSetup {angularLimitsMax=<10 10 10>, angularLimitsMin=<-10 -10 -10>, angularTarget=<1 0 0>, linearAxesMax=<0 0 1>, linearAxesMin=<0 0 -0.5>}
      localJointOffset=<1 0.5 0>
     }]
    relativeSpaceBone {boneName=root}
   }
   n32 = Anim Dynamics - Bone: dyn_iceCube_4!(ComponentPose=@n31, Alpha=1) -> (Pose) {
    angularDampingOverride=1
    angularSpringConstant=200
    bAngularSpring=true
    bLinearSpring=true
    bOverrideLinearDamping=true
    gravityScale=0
    linearDampingOverride=1
    linearSpringConstant=200
    simulationSpace=RootRelative
    alphaBoolBlend {blendOption=Linear}
    boundBone {boneName=dyn_iceCube_4}
    componentAppliedLinearAccClamp=<200 200 200>
    componentLinearAccScale=<1 1 1>
    componentLinearVelScale=<20 20 20>
    physicsBodyDefinitions [{
      boundBone {boneName=dyn_iceCube_4}
      boxExtents=<10 10 10>
      constraintSetup {angularLimitsMax=<10 10 10>, angularLimitsMin=<-10 -10 -10>, angularTarget=<1 0 0>, linearAxesMax=<0 0 1>, linearAxesMin=<0 0 -0.5>}
      localJointOffset=<1 0.5 0>
     }]
    relativeSpaceBone {boneName=root}
   }
   n33 = Anim Dynamics - Bone: dyn_iceCube_5!(ComponentPose=@n32, Alpha=1) -> (Pose) {
    angularDampingOverride=1
    angularSpringConstant=200
    bAngularSpring=true
    bLinearSpring=true
    bOverrideLinearDamping=true
    gravityScale=0
    linearDampingOverride=1
    linearSpringConstant=200
    simulationSpace=RootRelative
    alphaBoolBlend {blendOption=Linear}
    boundBone {boneName=dyn_iceCube_5}
    componentAppliedLinearAccClamp=<3000 3000 300>
    componentLinearAccScale=<0.3 0.3 0.3>
    componentLinearVelScale=<0.3 0.3 0.3>
    physicsBodyDefinitions [{
      boundBone {boneName=dyn_iceCube_5}
      boxExtents=<10 10 10>
      constraintSetup {angularLimitsMax=<10 10 10>, angularLimitsMin=<-10 -10 -10>, angularTarget=<1 0 0>, linearAxesMax=<0 0 4>, linearAxesMin=<0 0 -0.5>}
      localJointOffset=<1 0.5 0>
     }]
    relativeSpaceBone {boneName=root}
   }
   n34 = Anim Dynamics - Bone: dyn_iceCube_6!(ComponentPose=@n33, Alpha=1) -> (Pose) {
    angularDampingOverride=1
    angularSpringConstant=200
    bAngularSpring=true
    bLinearSpring=true
    bOverrideLinearDamping=true
    gravityScale=0
    linearDampingOverride=1
    linearSpringConstant=200
    simulationSpace=RootRelative
    alphaBoolBlend {blendOption=Linear}
    boundBone {boneName=dyn_iceCube_6}
    componentAppliedLinearAccClamp=<200 200 200>
    componentLinearAccScale=<1 1 1>
    componentLinearVelScale=<20 20 20>
    physicsBodyDefinitions [{
      boundBone {boneName=dyn_iceCube_6}
      boxExtents=<10 10 10>
      constraintSetup {angularLimitsMax=<10 10 10>, angularLimitsMin=<-10 -10 -10>, angularTarget=<1 0 0>, linearAxesMax=<0 0 1>, linearAxesMin=<0 0 -0.5>}
      localJointOffset=<1 0.5 0>
     }]
    relativeSpaceBone {boneName=root}
   }
   n35 = Anim Dynamics - Bone: dyn_iceCube_7!(ComponentPose=@n34, Alpha=1) -> (Pose) {
    angularDampingOverride=1
    angularSpringConstant=200
    bAngularSpring=true
    bLinearSpring=true
    bOverrideLinearDamping=true
    gravityScale=0
    linearDampingOverride=1
    linearSpringConstant=200
    simulationSpace=RootRelative
    alphaBoolBlend {blendOption=Linear}
    boundBone {boneName=dyn_iceCube_7}
    componentAppliedLinearAccClamp=<200 200 200>
    componentLinearAccScale=<1 1 1>
    componentLinearVelScale=<20 20 20>
    physicsBodyDefinitions [{
      boundBone {boneName=dyn_iceCube_7}
      boxExtents=<10 10 10>
      constraintSetup {angularLimitsMax=<10 10 10>, angularLimitsMin=<-10 -10 -10>, angularTarget=<1 0 0>, linearAxesMax=<0 0 1>, linearAxesMin=<0 0 -0.5>}
      localJointOffset=<1 0.5 0>
     }]
    relativeSpaceBone {boneName=root}
   }
   Save cached pose 'PreAnimDym'(Pose=Component To Local(ComponentPose=@n35)) {cacheName=PreAnimDym}
  ]}
 EventGraph {nodes [
   Event BlueprintUpdateAnimation()
   > n1 = Sequence!(execute) -> (then_0, then_1, then_2:Exec)
   > n2 = Gravity Override State Function DEPRECATED(CharacterPartAnimInstance=Self-Reference(), GameTimeInSeconds=GetGameTimeInSeconds(), GravityOverrideParams=Get Gravity Override Params())
   > Set GravityOverride(GravityOverride=@n2)
   n3 = GetOwnerSkeletalMeshComponent()
   n4 = BreakRotator(InRot=BreakTransform(InTransform=GetDeltaTransformFromRefPose(self=@n3, BoneName=spine_05, BaseName=spine_01)).Rotation)
   Set backpack_translation(execute=@n1.then_1, backpack_translation=MakeVector(Y=MapRangeClamped(Value=@n4.Yaw, InRangeA=0, InRangeB=-30, OutRangeB=-1.3)))
   > Set backpack_rotate(backpack_rotate=MakeRotator(Roll=MapRangeClamped(Value=@n4.Yaw, InRangeA=0, InRangeB=-30, OutRangeB=-5)))
   n5 = Get isMale()
   Set isFemaleAndFloating(execute=@n1.then_2, isFemaleAndFloating:bool=AND Boolean(A=NOT Boolean(A=@n5), B=Get bIsSkydiveFloating())) -> (Output_Get:bool)
   Event OnCharacterCustomizationLoaded()
   > n6 = Set Owning Pawn(Owning Pawn:Pawn Object Reference=TryGetFortPawn()) -> (Output_Get:Pawn Object Reference)
   > Is Valid(InputObject=@n6)
   > n7 = FindGender!(execute, Object:Object Reference=Get Owning Pawn() -> (Owning Pawn:Pawn Object Reference)) -> (True:Exec, False:Exec, OutGender:EFortCustomGender Enum=Invalid)
   > Set isMale(isMale:bool="Equal (Enum)"(A:EFortCustomGender Enum=@n7, B:EFortCustomGender Enum=Male)) -> (Output_Get:bool)
   TryGetPawnOwner()
  ]}
]
variables [
 backpack_rotate {type=Rotator}
 backpack_translation {type=Vector}
 Gravity Override Params {type=GravityOverrideParamsStruct}
 GravityOverride {type=Vector}
 isFemaleAndFloating {type=boolean}
 isMale {type=boolean}
 Owning Pawn {type=Pawn, default=null}
]
